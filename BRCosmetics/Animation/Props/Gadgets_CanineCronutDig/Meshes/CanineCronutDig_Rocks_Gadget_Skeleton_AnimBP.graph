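# Animation Blueprint for the CanineCronutDig rocks gadget prop skeleton, which hides the prop until an emote starts.
# While bHasEmoteStarted is false, the ref pose has its root bone scaled to zero; when true, it instantly switches to a non-looping sequence.
# That sequence is the MopTwirl emote's sign prop animation, not a CanineCronutDig one.
# The result feeds DefaultSlot for montages; the event graph is empty stubs and never sets the flag.
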AnimBlueprint /BRCosmetics/Animation/Props/Gadgets_CanineCronutDig/Meshes/CanineCronutDig_Rocks_Gadget_Skeleton_AnimBP
target_skeleton=&/CanineCronutDig_Rocks_Gadget_Skeleton
graphs [
 AnimGraph {nodes [
   n1 = "Transform (Modify) Bone - Bone: root"!(ComponentPose=Local To Component(LocalPose=Local Space Ref Pose()), Alpha, Scale="0.000,0.000,0.000") -> (Pose) {scaleMode=BMM_Replace, alphaBoolBlend {blendOption=Linear}, boneToModify {boneName=root}, scale=<0 0 0>}
   n2 = Component To Local(ComponentPose=@n1)
   n3 = Sequence Player 'Emote_Moptwirl_CMM_Prop_Sign'() {bLoopAnimation=false, sequence=/BRCosmetics/Animation/Game/MainPlayer/Emotes/MopTwirl/CMM/Emote_Moptwirl_CMM_Prop_Sign}
   n4 = Get bHasEmoteStarted() -> (bHasEmoteStarted:bool)
   Output Pose(Result=Slot 'DefaultSlot'(Source=Blend Poses by bool(BlendPose_0=@n3, BlendPose_1=@n2, BlendTime_0=0, BlendTime_1=0, bActiveValue=@n4) {blendType=Linear}))
  ]}
 EventGraph {nodes [
   TryGetPawnOwner()
   Event BlueprintUpdateAnimation()
  ]}
]
variables [bHasEmoteStarted {type=boolean}]
